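# Unlit, two-sided translucent material for the thruster effect on the DerangedMile glider, usable on Niagara sprites, ribbons and mesh particles.
# A procedural, tapered UV shape, optionally swizzled, combines with panning noise distorted by a normal map and driven by a Time dynamic parameter.
# A smoothstep of shape and noise blends colors A and B, tinted by particle color.
# Noise-eroded opacity is multiplied by view-angle fade, depth fade and particle alpha.
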Material /BRCosmetics/Gadgets/Assets/VinderTech_GliderChute/Glider_DerangedMile/FX/M_Glider_DerangedMile_Thrusters
graph {
 properties {
  bUsedWithNiagaraMeshParticles=true
  bUsedWithNiagaraRibbons=true
  bUsedWithNiagaraSprites=true
  blendMode=BLEND_Translucent
  pixelDepthOffsetMode=PDOM_Legacy
  previewMesh=/BRCosmetics/Accessories/FORT_Backpacks/Backpack_M_MED_SkippingClouds/FX/SM_Backpack_SkippingClouds_LeftEyeGlow
  shadingModel=MSM_Unlit
  twoSided=true
  usedShadingModels=Unlit
 }
 nodes [
  Constant() {r=1}
  n1 = VectorParameter() {parameterName=A, defaultValue=<rgba 1 0.0333 0 0>}
  n2 = VectorParameter() {parameterName=B, defaultValue=<rgba 15 6.04 1.56 0>}
  n3 = ScalarParameter() {defaultValue=0.17, parameterName=Color A Start}
  n4 = TextureCoordinate()
  n5 = OneMinus(None=MaterialFunctionCall(XY=@n4) {materialFunction=&Swizzle}.YX)
  n6 = Constant() {r=1}
  n7 = Constant() {r=0.1}
  n8 = Constant() {r=0.8}
  n9 = OneMinus(None=ComponentMask(None=TextureCoordinate()) {r=true})
  n10 = LinearInterpolate(A=Constant() {r=0.2}, B=Constant() {r=0.7}, Alpha=Power(Base=SmoothStep(Min=@n7, Max=@n8, Value=@n9), Exp=Constant() {r=2}))
  n11 = MaterialFunctionCall(Float2=MaterialFunctionCall(UVs=StaticSwitchParameter(True=@n5, False=@n4) {parameterName=Swize}, Texture Scale=AppendVector(A=@n6, B=@n10)) {materialFunction=&ScaleUVsByCenter}.UVs) {materialFunction=&BreakOutFloat2Components}
  n12 = Power(Base=Multiply(A=@n11.R), Exp=Constant() {r=1.5})
  n13 = ConstantBiasScale(None=Multiply(A=@n11.G)) {bias=-0.5, scale=2}
  n14 = Multiply(A=LinearInterpolate(A=Constant(), Alpha=OneMinus(None=@n12)) {constA=0.05}, B=OneMinus(None=Abs(None=@n13)))
  n15 = TextureCoordinate() {coordinateIndex=1}
  n16 = MaterialFunctionCall(Float2=@n15) {materialFunction=&BreakOutFloat2Components}
  n17 = Constant() {r=1.5}
  n18 = VectorParameter() {parameterName=UV, defaultValue=<rgba 0.8 5 -2 0.2>}
  n19 = AppendVector(A=@n18.R, B=@n18.R)
  n20 = Multiply(A=AppendVector(A=Power(Base=@n16.R, Exp=@n17), B=@n16.G), B=@n19)
  n21 = DynamicParameter() {paramNames [Time,Param2,Param3,Param4]}
  n22 = AppendVector(A=@n18.B, B=@n18.B)
  n23 = Multiply(A=Multiply(A=@n15) {constB=0.5}, B=@n19)
  n24 = Multiply(A=@n22) {constB=0.35}
  n25 = ComponentMask(None=TextureSample(UVs=Panner(Coordinate=@n23, Time=@n21.Time, Speed=@n24)) {samplerType=SAMPLERTYPE_Normal, texture=&T_Weap_Distort_N_02}.RGB) {g=true, r=true}
  n26 = TextureSample(UVs=Add(A=Panner(Coordinate=@n20, Time=@n21.Time, Speed=@n22), B=Multiply(A=@n25) {constB=0.35})) {samplerType=SAMPLERTYPE_LinearColor, texture=&T_NoisePack_LinearColor}
  n27 = Multiply(A=@n14, B=@n26.R)
  n28 = Multiply(A=Multiply(A=Add(A=@n14, B=@n27), B=Constant() {r=0.25}) {constB=0.25}, B=Constant() {r=1})
  n29 = TextureCoordinate()
  n30 = SmoothStep(Min=Constant(), Max=Constant() {r=0.8}, Value=OneMinus(None=ComponentMask(None=@n29) {r=true}))
  n31 = SmoothStep(
   Min=@n3
   Max=Add(A=@n3, B=ScalarParameter() {defaultValue=0.6, parameterName=Color Transition Range})
   Value=Saturate(None=Multiply(A=@n28, B=@n30))
  )
  MaterialFunctionCall(Emissive=Multiply(A=ParticleColor().RGB, B=LinearInterpolate(A=@n1.RGB, B=@n2.RGB, Alpha=@n31))) {materialFunction=&MF_FX_EmissiveAdjustments}
  n32 = Power(Base=OneMinus(None=NamedRerouteUsage()), Exp=Constant() {r=2})
  n33 = NamedRerouteUsage()
  n34 = Subtract(A=@n33, B=Multiply(A=Subtract(A=@n32, B=@n33), B=ScalarParameter() {defaultValue=1.5, parameterName=Amount}))
  n35 = Multiply(
   A=NamedRerouteDeclaration(None=Saturate(None=@n34)) {name=Opacity, nodeColor=<rgba 0.3125 0 1 1>}
   B=ScalarParameter() {defaultValue=1, parameterName=MainOpacityMult}
  )
  n36 = MaterialFunctionCall("FadeSensitivity (optional)"=@n6) {materialFunction=&FadeBasedOnViewAngle}
  n37 = ScalarParameter() {defaultValue=25, parameterName=DethFade}
  Multiply(A=Multiply(A=Multiply(A=@n35, B=@n36.Result), B=DepthFade(FadeDistance=@n37)), B=ParticleColor().A)
  Multiply(A=@n16.R, B=@n16.R)
  NamedRerouteDeclaration(None=@n14) {name=BaseShape, nodeColor=<rgba 0 0.6484 1 1>}
  NamedRerouteDeclaration(None=@n26.R) {name=BaseNoise, nodeColor=<rgba 0.9688 1 0 1>}
 ]
}
