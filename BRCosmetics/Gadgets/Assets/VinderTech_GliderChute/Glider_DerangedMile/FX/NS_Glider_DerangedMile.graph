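# Niagara system for the Deranged Mile glider's effects, including electrical sparks, flung mesh chunks, glows, flickering headlights, lingering smoke and further sparks.
# Emitters place particles through a shared skeletal socket reader, choosing sockets by random pick or spawn group.
# A system Timeline modulates rates, forces and tints, while User parameters select color variant and front-end versus gameplay behavior.
# Glows and headlights also scale with time of day.
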NiagaraSystem /BRCosmetics/Gadgets/Assets/VinderTech_GliderChute/Glider_DerangedMile/FX/NS_Glider_DerangedMile
dependencies {
 usedDataInterfaces [
  &NiagaraDataInterfaceCurve
  &NiagaraDataInterfaceArrayFloat3
  &NiagaraDataInterfaceColorCurve
  &NiagaraDataInterfaceArrayInt32
  &NiagaraDataInterfacePlatformSet
  &NiagaraDataInterfaceMeshRendererInfo
  &NiagaraDataInterfaceVector2DCurve
  &NiagaraDataInterfaceSocketReader
 ]
 usedDynamicInputs []
 usedModules [
  &self:SystemSpawnScript.SystemScriptSource.SystemScriptGraph.NiagaraNodeAssignment_4.SetVariables_#1
  &self:SystemSpawnScript.SystemScriptSource.SystemScriptGraph.NiagaraNodeAssignment_5.SetVariables_#2
  &SystemState
  &self:SystemSpawnScript.SystemScriptSource.SystemScriptGraph.NiagaraNodeAssignment_2.SetVariables_#3
  &self:SystemSpawnScript.SystemScriptSource.SystemScriptGraph.NiagaraNodeAssignment_3.SetVariables_#4
  &Timeline
  &self:SystemSpawnScript.SystemScriptSource.SystemScriptGraph.NiagaraNodeAssignment_6.SetVariables_#5
  &self:SystemSpawnScript.SystemScriptSource.SystemScriptGraph.NiagaraNodeAssignment_7.SetVariables_#6
  &self:SystemSpawnScript.SystemScriptSource.SystemScriptGraph.NiagaraNodeAssignment_8.SetVariables_#7
  &self:SystemSpawnScript.SystemScriptSource.SystemScriptGraph.NiagaraNodeAssignment_9.SetVariables_#8
 ]
}
emitters {
 ElectricalSparks {
  emitterUpdateScript [
   /SetVariables(
    Emitter.RandSocket:NiagaraInt32=&RandomRangeInt(Minimum=4, Maximum=9)
    Emitter.RandSocketFE:NiagaraInt32=&SelectIntFromArray(
     Array Sampling Mode=::NewEnumerator1
     Direct Array Index=?
     Recalculate Random Each Loop=true
     Int Selection Array=DataInterface() {propertyValues {IntData [4,5,6,7,8,9,12,12]}}
     Random Seed=0
    )
   ) {isSetParameters=true}
   /SetVariables(Emitter.RandSeed:NiagaraInt32=&RandomRangeInt(Minimum=0, Maximum=2147483647)) {isSetParameters=true}
   /SetVariables(
    Emitter.IdleSpawnTime:NiagaraFloat=&RandomRangeFloat(Minimum=0, Maximum=0.5, Random Seed=$Emitter.RandSeed, Fixed Random Seed=true)
    Emitter.ReactiveSpawnRate=100
   ) {isSetParameters=true}
   &EmitterState(
    Life Cycle Mode=::NewEnumerator1
    Loop Duration=&Lerp_Float(A=0.5, B=0.15, Alpha=$System.Timeline.Time)
    UseLoopDelay=false
   )
   &SpawnBurst_Instantaneous(
    Spawn Count=&RandomRangeInt(Minimum=20, Maximum=26)
    Spawn Time=$Emitter.IdleSpawnTime
    Use Spawn Probability=true
    UseLoopCountLimit=false
   ) {module=SpawnBurst_Instantaneous002}
   &SpawnBurst_Instantaneous(Spawn Group=1, Spawn Time=$Emitter.IdleSpawnTime, Use Spawn Probability=true, UseLoopCountLimit=false) {module=SpawnBurst_Instantaneous003}
  ]
  particleSpawnScript [
   /SetVariables(Particles.VisibilityTag=$Emitter.SpawnGroup) {isSetParameters=true}
   /SetVariables(Particles.PerParticleRandom:NiagaraFloat=&FloatFromCurve(
     FloatCurve=DataInterface() {propertyValues {Curve {keys [
         {arriveTangent=4.1785, interpMode=RCIM_Cubic, tangentMode=RCTM_Break, time=0, value=0}
         {arriveTangent=3.1227, interpMode=RCIM_Cubic, leaveTangent=0.0224, tangentMode=RCTM_Break, time=1, value=1}
        ]}}}
     CurveIndex=&RandomRangeFloat(Minimum=0, Maximum=1)
    )) {isSetParameters=true}
   &InitializeParticle(
    Color Mode=::NewEnumerator0
    Lifetime=&RandomRangeFloat(Minimum=0.4, Maximum=0.9)
    UsePositionOffset=false
    Write Material Random=true
   )
   &SocketLocation(
    Interpolate Socket Sampling=true
    SamplePosOffset=false
    Sampled Orientation Coordinate Space:ENiagaraCoordinateSpace=::Local
    Sampled Position Coordinate Space:ENiagaraCoordinateSpace=::World
    Socket Index=&MakeCustomIntFromBool(
     Bool=&Boolean_OrOperation(Bool1=$User.bIsFrontEnd, Bool2=$User.bIsFrontEndPreview)
     True INT=$Emitter.RandSocketFE
     False INT=$Emitter.RandSocket
    )
    Socket Reader=$System.Socket Reader
    Socket Sampling Mode=::NewEnumerator3
    VelocitySpeedLimit=false
   )
   &AddVelocity(
    Additional Quaternion Rotation=false
    Cone Angle=&RandomRangeFloat(Minimum=25, Maximum=90)
    Cone Axis=&TransformVector(Vector=<1 0 0>, Destination Space=::World)
    Module.UseConeFalloff=false
    Module.UseVelDistribution=false
    Override Seed=false
    Rotation Mode=::NewEnumerator3
    Rotation Quaternion:Quat4f=$Output.SocketLocation.SampledRotation
    Velocity Mode=::NewEnumerator2
    Velocity Speed=&RandomRangeFloat(Minimum=150, Maximum=700)
   )
   &AddVelocity(
    Velocity=&TransformVector(Vector=<0 0 1>, Destination Space=::World)
    Velocity Speed Scale=&Multiply_Float(
     A=&RandomRangeFloat(Minimum=10, Maximum=40)
     B=&FN_FrontEndFloatModifier(Gameplay=1, Front End=2)
    )
   ) {module=AddVelocity001}
   /SetVariables(
    Particles.FlareSpriteSize:Vector2f=&Multiply_Vector2D_ByFloat(Float=&Multiply_Float(A=3, B=$System.OwnerScale), Vector2D=<50 50>)
    Particles.GlowSpriteSize:Vector2f=&Multiply_Vector2D_ByFloat(Float=1.4, Vector2D=&Lerp_Vector2(A=<5 10>, B=<10 20>, Alpha=$Particles.PerParticleRandom))
    Particles.SparkSpriteSize:Vector2f=&Multiply_Vector2D_ByFloat(Float=1.4, Vector2D=&Lerp_Vector2(A=<1 2>, B=<2 7>, Alpha=$Particles.PerParticleRandom))
   ) {isSetParameters=true}
   &CameraOffset(Camera Offset Amount=&MaskFloatBySpawnGroup(Masked Float=0, Passthrough Float=12))
  ]
  particleUpdateScript [
   &SocketLocation(
    Interpolate Socket Sampling=true
    Orientation Sampling=::NewEnumerator1
    Position Sampling=::NewEnumerator1
    SamplePosOffset=false
    Sampled Orientation Coordinate Space:ENiagaraCoordinateSpace=::Local
    Sampled Position Coordinate Space:ENiagaraCoordinateSpace=::World
    Socket Index=&MakeCustomIntFromBool(
     Bool=&Boolean_OrOperation(Bool1=$User.bIsFrontEnd, Bool2=$User.bIsFrontEndPreview)
     True INT=$Emitter.RandSocketFE
     False INT=$Emitter.RandSocket
    )
    Socket Reader=$System.Socket Reader
    Socket Sampling Mode=::NewEnumerator3
    VelocitySpeedLimit=false
   ) {module=SocketLocation003}
   /SetVariables(Particles.LightRadius=&MaskFloat(
     Mask Source=1
     Mask Parameter=&MakeCustomIntFromBool(
      Bool=&Boolean_OrOperation(Bool1=$User.bIsFrontEnd, Bool2=$User.bIsFrontEndPreview)
      True INT=$Emitter.RandSocketFE
      False INT=$Emitter.RandSocket
     )
     Mask=12
     Masked Float=33
     Passthrough Float=100
    )) {isSetParameters=true}
   /SetVariables(Particles.GlowPosition:NiagaraPosition=$Output.SocketLocation003.SampledPosition) {isSetParameters=true}
   /SetVariables(Particles.Position_LightGlow:NiagaraPosition=$Particles.GlowPosition) {isSetParameters=true}
   /SetVariables(Particles.SparkSpriteColor:LinearColor=&MultiplyLinearColor(
     Linear Color=&ColorFromCurve(ColorCurve=DataInterface() {propertyValues {
        AlphaCurve {keys [<at 0.01 1>,<at 0.9198 1 RCIM_Cubic>,<at 1 0>]}
        BlueCurve {keys [<at 0 15>,<at 0.0135 0.5>,<at 0.2072 0>,<at 0.4906 0>,<at 0.7101 0>,<at 0.9394 0>]}
        GreenCurve {keys [<at 0 4.5>,<at 0.0135 3.8>,<at 0.2072 2.2>,<at 0.4906 1.0705>,<at 0.7101 0.1833>,<at 0.9394 0.0762>]}
        RedCurve {keys [<at 0 0>,<at 0.0135 5>,<at 0.2072 3>,<at 0.4906 2>,<at 0.7101 1>,<at 0.9394 1>]}
       }})
     Scale Factor=&MakeFloat4FromVectorAndFloat("Vector (RGB)"=<1 1 1>, "Float (Alpha)"=0.05)
    )) {isSetParameters=true}
   /SetVariables(Particles.GlowSpriteColor:LinearColor=&MultiplyLinearColor(
     Linear Color=$Particles.SparkSpriteColor
     Scale Factor=&MakeFloat4FromVectorAndFloat("Vector (RGB)"=<1 1 1>, "Float (Alpha)"=0.1)
    )) {isSetParameters=true}
   /SetVariables(Particles.FlareSpriteColor:LinearColor=&MultiplyLinearColor(
     Linear Color=&ColorFromCurve(ColorCurve=DataInterface() {propertyValues {
        AlphaCurve {keys [<at 0.01 1>,<at 0.9198 1 RCIM_Cubic>,<at 1 0>]}
        BlueCurve {keys [<at 0 5>,<at 0.0963 0.1>,<at 0.3587 0>,<at 0.6339 0>,<at 1 0>]}
        GreenCurve {keys [<at 0 1.5>,<at 0.0963 0.76>,<at 0.3587 2.2>,<at 0.6339 1.0705>,<at 1 0.1833>]}
        RedCurve {keys [<at 0 0>,<at 0.0963 1>,<at 0.3587 3>,<at 0.6339 2>,<at 1 1>]}
       }})
     Scale Factor=&MakeFloat4FromVectorAndFloat(
      "Vector (RGB)"=&VectorFromFloat(Value=&FN_FrontEndFloatModifier(Gameplay=0.5, Front End=0.75))
      "Float (Alpha)"=&FloatFromCurve(
       FloatCurve=DataInterface() {propertyValues {Curve {keys [
           {tangentMode=RCTM_User, time=0, value=1}
           <at 0.0756 -0.7485>
           <at 0.1284 2.2279>
           <at 0.3098 -0.7559>
           <at 0.4525 0.9206>
           <at 0.525 0.0265>
           <at 0.6661 0.3676>
           <at 0.7436 1.6548>
           <at 0.8248 0.1441>
           <at 0.9286 0.8147>
           {tangentMode=RCTM_User, time=1, value=0}
          ]}}}
       CurveIndex=&NormalizeFloat(
        Value=&Multiply_Float(A=$Particles.NormalizedAge, B=&RandomRangeFloat(Minimum=2.5, Maximum=4))
        Min=0
        Max=1
       )
       Scale Curve=0.05
      )
     )
    )) {isSetParameters=true}
   /SetVariables(Particles.LightColor=&MultiplyLinearColor(
     Linear Color=$Particles.FlareSpriteColor
     Scale Factor=&MakeFloat4FromVectorAndFloat("Vector (RGB)"=&VectorFromFloat(Value=5), "Float (Alpha)"=5)
    )) {isSetParameters=true}
   &ParticleState()
   &Drag(
    Drag=&FloatFromCurve(
     FloatCurve=DataInterface() {propertyValues {Curve {keys [{interpMode=RCIM_Cubic, tangentMode=RCTM_User, time=0, value=1}, {interpMode=RCIM_Cubic, tangentMode=RCTM_User, time=1, value=0}]}}}
     Scale Curve=&Lerp_Float(A=17, B=5, Alpha=$Particles.PerParticleRandom)
    )
    Use Linear Drag=true
    Use Rotational Drag=true
   )
   &GravityForce(Gravity=&MaskVectorBySpawnGroup(
     Masked Vector=&FN_FrontEndVector3Modifier(Gameplay=<0 0 -980>, Front End=<0 0 -980>)
     Passthrough Vector=<0 0 0>
    ))
   &InheritSourceMovement(
    Applied Position Scale=&VectorFromFloat(Value=&MergeHelper_FloatSwitch(
      0=&FN_FrontEndFloatModifier(
       Gameplay=&FloatFromCurve(FloatCurve=DataInterface() {propertyValues {Curve {keys [<at 0.5 1>,<at 1 0>]}}})
       Front End=1
      )
      1=1
      2=?
      3=?
      Index=$Emitter.SpawnGroup
      Num=::NewEnumerator0
     ))
    ApplyForce=false
    ApplyPosition=true
    ApplyPositionFalloff=false
   )
   &AccelerationForce(Acceleration=&TransformVector(Vector=<0 -700 0>, Destination Space=::Simulation))
   &SolveForcesAndVelocity(Clamp Velocity=false, Limit Acceleration=false)
  ]
  rendererClasses [&NiagaraSpriteRendererProperties, &NiagaraLightRendererProperties]
  renderers [
   {rendererClass=&NiagaraSpriteRendererProperties}
   {rendererClass=&NiagaraSpriteRendererProperties, rendererIndex=1}
   {rendererClass=&NiagaraLightRendererProperties, rendererIndex=2}
   {rendererClass=&NiagaraSpriteRendererProperties, rendererIndex=3}
  ]
 }
 GliderChunks {
  emitterUpdateScript [
   &EmitterState(Life Cycle Mode=::NewEnumerator1, Loop Duration=1, UseLoopDelay=false)
   &SpawnRate(SpawnRate=100, Use Spawn Probability=false) {enabled=false}
   /SetVariables(
    Emitter.RandSocket:NiagaraInt32=&RandomRangeInt(Minimum=0, Maximum=1)
    Emitter.RandSocketFE:NiagaraInt32=&RandomRangeInt(Minimum=0, Maximum=2)
   ) {isSetParameters=true}
   &SpawnBurst_Instantaneous(
    Spawn Count=&RandomRangeInt(Minimum=3, Maximum=4)
    Use Spawn Probability=false
    UseLoopCountLimit=false
   )
  ]
  particleSpawnScript [
   &FN_SetSpawnGroup(Number of Spawn Groups=3)
   &InitializeParticle(
    Lifetime Min=2
    Lifetime Mode=::NewEnumerator1
    Mesh Renderer Array Visibility Mode=::NewEnumerator3
    Mesh Renderer Info=DataInterface() {propertyValues {MeshRenderer=&self:GliderChunks_0.NiagaraMeshRendererProperties_0}}
    Mesh Scale Mode=::NewEnumerator3
    Mesh Uniform Scale=&Multiply_Float(
     A=&RandomRangeFloat(Minimum=0.5, Maximum=0.8)
     B=&Multiply_Float(A=0.15, B=$System.OwnerScale)
    )
    UsePositionOffset=true
    Write Material Random=true
   )
   &SocketLocation(
    Interpolate Socket Sampling=true
    SamplePosOffset=false
    Sampled Orientation Coordinate Space:ENiagaraCoordinateSpace=::Local
    Sampled Position Coordinate Space:ENiagaraCoordinateSpace=::World
    Socket Index=&MergeHelper_IntSwitch(
     0=5
     1=8
     2=12
     3=?
     4=?
     5=?
     Index=&MakeCustomIntFromBool(
      Bool=&Boolean_OrOperation(Bool1=$User.bIsFrontEnd, Bool2=$User.bIsFrontEndPreview)
      True INT=$Emitter.RandSocketFE
      False INT=$Emitter.RandSocket
     )
     Num=::NewEnumerator1
    )
    Socket Reader=$System.Socket Reader
    Socket Sampling Mode=::NewEnumerator3
    VelocitySpeedLimit=false
   )
   &AddVelocity(
    Additional Quaternion Rotation=false
    Cone Angle=&RandomRangeFloat(Minimum=25, Maximum=90)
    Cone Axis=&TransformVector(
     Vector=&FN_FrontEndVector3Modifier(
      Gameplay=&MergeHelper_VectorSwitch(
       0=<1 0 1.3>
       1=<1 0 1.3>
       2=<0 1 1>
       3=?
       Index=&MakeCustomIntFromBool(
        Bool=&Boolean_OrOperation(Bool1=$User.bIsFrontEnd, Bool2=$User.bIsFrontEndPreview)
        True INT=$Emitter.RandSocketFE
        False INT=$Emitter.RandSocket
       )
       Num=::NewEnumerator1
      )
      Front End=&MergeHelper_VectorSwitch(
       0=<-1 0 1.3>
       1=<1 0 1.3>
       2=<0 1 1>
       3=?
       Index=&MakeCustomIntFromBool(
        Bool=&Boolean_OrOperation(Bool1=$User.bIsFrontEnd, Bool2=$User.bIsFrontEndPreview)
        True INT=$Emitter.RandSocketFE
        False INT=$Emitter.RandSocket
       )
       Num=::NewEnumerator1
      )
     )
     Destination Space=::World
    )
    Module.UseConeFalloff=false
    Module.UseVelDistribution=false
    Override Seed=false
    Rotation Mode=::NewEnumerator3
    Rotation Quaternion:Quat4f=$Output.SocketLocation.SampledRotation
    Velocity Mode=::NewEnumerator2
    Velocity Speed=&RandomRangeFloat(Minimum=250, Maximum=500)
   )
   /SetVariables(Particles.PerParticleRandom:NiagaraFloat=&RandomRangeFloat(Minimum=0, Maximum=1)) {isSetParameters=true}
   &InitialMeshOrientation(
    Mesh Orientation Mode=::NewEnumerator2
    Rotation=&RandomRangeVector(Minimum=<0 0 0>, Maximum=<1 1 1>)
    Use Rotation Vector=true
   )
   /SetVariables(Particles.SpawnRotationVector=&Multiply_VectorByFloat(Vector=&RandomRangeVector(Minimum=<0.5 -1 0.5>, Maximum=<-0.5 -1 -0.5>), Float=1)) {isSetParameters=true}
   /SetVariables(Particles.MeshIndex=&MergeHelper_IntSwitch(0=3, 1=2, 2=1, 3=?, 4=?, 5=?, Index=$Particles.SpawnGroup, Num=::NewEnumerator1)) {isSetParameters=true}
  ]
  particleUpdateScript [
   &ParticleState() {module=ParticleState001}
   &Drag(
    Drag=&FloatFromCurve(FloatCurve=DataInterface() {propertyValues {Curve {keys [
         {arriveTangent=2.3591, interpMode=RCIM_Cubic, leaveTangent=2.3591, tangentMode=RCTM_User, time=0, value=0}
         {interpMode=RCIM_Cubic, tangentMode=RCTM_User, time=1, value=1}
        ]}}})
    Rotational Drag=1.3
    Use Linear Drag=true
    Use Rotational Drag=true
   )
   &GravityForce(Gravity=<0 0 -2000>)
   &InheritSourceMovement(ApplyForce=false, ApplyPosition=true, ApplyPositionFalloff=false) {module=InheritSourceMovement001}
   &AccelerationForce(Acceleration=&Multiply_VectorByFloat(
     Vector=&TransformVector(Vector=<0 1 0>, Destination Space=::World)
     Float=&Multiply_Float(A=2000, B=$System.Timeline.Time)
    ))
   &SolveForcesAndVelocity(Clamp Velocity=false, Limit Acceleration=false)
   &UpdateMeshOrientation(Rotation Rate=-1, Rotation Vector=$Particles.SpawnRotationVector)
   &ParticleState()
   &ScaleMeshSize(Scale Factor=&VectorFromFloat(Value=&FloatFromCurve(FloatCurve=DataInterface() {propertyValues {Curve {keys [<at 0 0.2>, <at 0.1 1>, {interpMode=RCIM_Cubic, leaveTangent=-17.6817, tangentMode=RCTM_Break, time=0.85, value=1}, <at 1 0>]}}})))
  ]
  rendererClasses [&NiagaraMeshRendererProperties]
  renderers [{rendererClass=&NiagaraMeshRendererProperties, rendererIndex=0}]
 }
 Glows {
  emitterUpdateScript [
   &EmitterState(
    Life Cycle Mode=::NewEnumerator1
    Loop Behavior=::NewEnumerator1
    Loop Delay=$System.InitialThursterDelay
    Loop Duration Mode=::NewEnumerator1
    UseLoopDelay=true
   )
   &SpawnBurst_Instantaneous(Spawn Count=8, Use Spawn Probability=false, UseLoopCountLimit=true)
  ]
  particleSpawnScript [
   &PartitionParticles(Number Of Partitions=8)
   &InitializeParticle(
    Color=&LerpLinearColors(StartColor=<rgba 1 0.0167 0 1>, EndColor=<rgba 0.1658 0.3048 1 1>, LerpFactor=$User.Variant)
    Mesh Scale Mode=::NewEnumerator3
    Mesh Uniform Scale=3
    Sprite Rotation Mode=::NewEnumerator0
    Sprite Size=&Multiply_Vector2D(
     A=&MergeHelper_Vector2Switch(
      0=<14 40>
      1=<14 37>
      2=<14 62>
      3=<14 62>
      4=<14 43>
      5=<14 60>
      6=<14 43>
      7=<14 60>
      Index=$Particles.PartitionParticles.Partition
      Num=::NewEnumerator6
     )
     B=&Vector2DFromFloat(Value=$System.OwnerScale)
    )
    Sprite Size Mode=::NewEnumerator4
    UsePositionOffset=false
    Write Material Random=true
   )
   &SpriteFacingAndAlignment(
    Sprite Alignment=&Multiply_Vector(
     A=&MergeHelper_VectorSwitch(
      0=<1 0 0>
      1=<1 0 0>
      2=<0 1 0>
      3=<0 1 0>
      4=<-0.05 1 0>
      5=<-0.44 0.9 0>
      6=<0.05 1 0>
      7=<0.44 1 0>
      Index=$Particles.PartitionParticles.Partition
      Num=::NewEnumerator6
     )
     B=<1 -1 1>
    )
    Sprite Facing=<0.7068 0.07 0.3>
    WriteAlignment=true
    WriteFacing=false
   )
  ]
  particleUpdateScript [
   &ParticleState(Kill Particles When Lifetime Has Elapsed=false, Let Infinitely Lived Particles Die When Emitter Deactivates=true)
   &SocketLocation(
    Interpolate Socket Sampling=true
    Orientation Sampling=::NewEnumerator1
    Position Sampling=::NewEnumerator1
    SamplePosOffset=false
    Sampled Position Coordinate Space:ENiagaraCoordinateSpace=::Local
    Socket Index=11
    Socket Reader=$System.Socket Reader
    Socket Sampling Mode=::NewEnumerator3
    VelocitySpeedLimit=false
   )
   &SystemLocation(
    MaskSpawnGroup=false
    Offset=&AddVectorToPosition(
     Position=$Output.SocketLocation.SampledPosition
     Vector=&RotateVectorByQuaternion(
      Vector To Rotate=&Multiply_Vector(
       A=&MergeHelper_VectorSwitch(
        0=<0 -141 -37>
        1=<0 146 -58>
        2=<-53 41 -26.5>
        3=<53 41 -26.5>
        4=<-60.5 -60 -32.5>
        5=<-49.5 -99 -35>
        6=<60.5 -60 -32.5>
        7=<49.5 -99 -35>
        Index=$Particles.PartitionParticles.Partition
        Num=::NewEnumerator6
       )
       B=<1 -1 1>
      )
      Quaternion=$Output.SocketLocation.SampledRotation
     )
    )
   )
   &DynamicMaterialParameters(
    Index 0 Param 1=2
    Module.Index0Param0Write=true
    Module.Index0Param1Write=false
    Module.Index0Param2Write=false
    Module.Index0Param3Write=false
   )
   &FN_Mesh_Attributes(Create Color Attribute=true, Write to Module Namespace=true)
   &CameraOffset(Camera Offset Amount=3)
   &ScaleColor(
    Scale Alpha=&FloatFromCurve(
     FloatCurve=DataInterface() {propertyValues {Curve {keys [<at 0 0>,<at 0.35 1>]}}}
     Scale Curve=&Lerp_Float(A=1.1, B=2, Alpha=$System.Timeline.Time)
    )
    ScaleA=true
    ScaleRGB=false
   )
   &ScaleColor(
    Scale Alpha=&FNTimeOfDayScalar(Night Scale=0.35, Morning Scale=0.3, Day Scale=0.35, Evening Scale=0.5)
    ScaleA=true
    ScaleRGB=false
   ) {module=ScaleColor001}
  ]
  rendererClasses [&NiagaraSpriteRendererProperties]
  renderers [{rendererClass=&NiagaraSpriteRendererProperties}]
 }
 Headlights {
  emitterUpdateScript [
   &EmitterState(
    Inactive Response=::NewEnumerator1
    Life Cycle Mode=::NewEnumerator1
    Loop Behavior=::NewEnumerator1
    Loop Delay=$System.InitialThursterDelay
    Loop Duration Mode=::NewEnumerator1
    UseLoopDelay=true
   )
   &SpawnBurst_Instantaneous(Spawn Count=2, Use Spawn Probability=false, UseLoopCountLimit=false)
  ]
  particleSpawnScript [
   &FN_SetSpawnGroup(Number of Spawn Groups=2)
   /SetVariables(Particles.MirrorAxis:Vector3f=&MaskVectorBySpawnGroup(Masked Vector=<-1 1 1>, Passthrough Vector=<1 1 1>)) {isSetParameters=true}
   &InitializeParticle(
    Color=<rgba 1 0.55 0.1 1>
    Mesh Scale Mode=::NewEnumerator3
    Sprite Size=&Multiply_Vector2D_ByFloat(Float=$System.OwnerScale, Vector2D=<250 500>)
    Sprite Size Mode=::NewEnumerator4
    UsePositionOffset=false
    Write Material Random=true
   )
   &CameraOffset(Camera Offset Amount=11)
   &DynamicMaterialParameters(
    Index 0 Param 1=&Multiply_Float(A=-19, B=&MakeFloatFromVector(Vector=$Particles.MirrorAxis, Channel=::NewEnumerator0))
    Index 0 Param 2=-1
    Module.Param0WriteEnabled=true
    Module.Param1WriteEnabled=true
    Module.Param2WriteEnabled=false
    Module.Param3WriteEnabled=false
   )
  ]
  particleUpdateScript [
   &ParticleState(Kill Particles When Lifetime Has Elapsed=false)
   &SocketLocation(
    Orientation Sampling=::NewEnumerator1
    SamplePosOffset=true
    Socket Index=&MergeHelper_IntSwitch(0=2, 1=3, 2=?, 3=?, 4=?, 5=?, Index=$Particles.SpawnGroup, Num=::NewEnumerator0)
    Socket Reader=$System.Socket Reader
    Socket Sampling Mode=::NewEnumerator3
    VelocitySpeedLimit=false
   ) {module=SocketLocation001}
   &SpriteFacingAndAlignment(
    Alignment Coordinate Space=::Local
    Sprite Alignment=&RotateVectorByQuaternion(Vector To Rotate=<0 -1 0>, Quaternion=$Output.SocketLocation001.SampledRotation)
    WriteAlignment=true
    WriteFacing=false
   ) {module=SpriteFacingAndAlignment001}
   /SetVariables(
    Particles.BeamColor:LinearColor=&MultiplyLinearColor(Linear Color=$Particles.Initial.Color, Scale Factor=&MakeVector4_FromFloat(Float=0.5))
    Particles.BeamPosition:NiagaraPosition=&AddVectorToPosition(
     Position=$Particles.Position
     Vector=&Multiply_Vector(A=&FN_FrontEndVector3Modifier(Gameplay=<0 10 2>, Front End=<0 -5 2>), B=$Particles.MirrorAxis)
    )
   ) {isSetParameters=true}
   /SetVariables(
    Particles.FlareAlignment:Vector3f=&Multiply_Vector(A=<-0.3 0 1>, B=$Particles.MirrorAxis)
    Particles.FlareCameraOffset=77
    Particles.FlareColor=&MultiplyLinearColor(Linear Color=$Particles.Initial.Color, Scale Factor=&MakeVector4_FromFloat(Float=1.666))
    Particles.FlareFacing:Vector3f=&Multiply_Vector(A=<0.6 -1 0.5>, B=$Particles.MirrorAxis)
    Particles.FlarePosition:NiagaraPosition=&AddVectorToPosition(Position=$Particles.Position, Vector=&Multiply_Vector(A=<0 0 0>, B=$Particles.MirrorAxis))
    Particles.FlareSize:Vector2f=<100 50>
   ) {isSetParameters=true}
   /SetVariables(
    Particles.TeardropGlowColor:LinearColor=&MultiplyLinearColor(Linear Color=$Particles.Initial.Color, Scale Factor=&MakeVector4_FromFloat(Float=1))
    Particles.TeardropGlowPosition:NiagaraPosition=&AddVectorToPosition(Position=$Particles.Position, Vector=&Multiply_Vector(A=<11 0 2>, B=$Particles.MirrorAxis))
    Particles.TeardropGlowProperties:Vector4f=&MakeFloat4FromVectorAndFloat(
     "Vector (RGB)"=&Multiply_Vector(
      A=&NormalizeVector(
       Vector To Normalize=/Niagara/DynamicInputs/Vectors/CrossProduct(Vector A:Vector3f=$Particles.FlareAlignment, Vector B:Vector3f=$Particles.FlareFacing, Normalize Incoming Vectors:NiagaraBool=true)
       Safe Normalize=true
       Safe Fallback Threshold=0
       Safe Fallback Vector=<0 0 0.001>
      )
      B=&VectorFromFloat(Value=&MakeFloatFromVector(Vector=$Particles.MirrorAxis, Channel=::NewEnumerator0))
     )
     "Float (Alpha)"=25
    )
    Particles.TeardropGlowSize:Vector2f=&Vector2DFromFloat(Value=22)
   ) {isSetParameters=true}
   /SetVariables(
    Particles.GlowColor=&MultiplyLinearColor(Linear Color=$Particles.Initial.Color, Scale Factor=&MakeVector4_FromFloat(Float=1))
    Particles.GlowPosition:NiagaraPosition=&AddVectorToPosition(Position=$Particles.Position, Vector=&Multiply_Vector(A=<0 0 0>, B=$Particles.MirrorAxis))
    Particles.GlowSize=&Vector2DFromFloat(Value=55)
   ) {isSetParameters=true}
   /SetVariables(
    Particles.LightColor=&MultiplyLinearColor(Linear Color=$Particles.Initial.Color, Scale Factor=&MakeVector4_FromFloat(Float=2))
    Particles.LightPosition:NiagaraPosition=&AddVectorToPosition(Position=$Particles.Position, Vector=&Multiply_Vector(A=<-2 -5 1>, B=$Particles.MirrorAxis))
   ) {isSetParameters=true}
   /SetVariables(Particles.Flicker:NiagaraFloat=&Multiply_Float(
     A=&Multiply_Float(
      A=1.2
      B=&FloatFromCurve(
       FloatCurve=DataInterface() {propertyValues {Curve {keys [<at 0 1 RCIM_Constant>,<at 0.333 0.6 RCIM_Constant>,<at 0.37 1 RCIM_Constant>,<at 0.7 0.7 RCIM_Constant>,<at 0.72 1 RCIM_Constant>,<at 1 1 RCIM_Constant>]}}}
       CurveIndex=&Add_Float(
        A=$System.NormalizedLoopAge
        B=&MergeHelper_FloatSwitch(0=0, 1=0.3, 2=?, 3=?, Index=$Particles.SpawnGroup, Num=::NewEnumerator0)
       )
      )
     )
     B=&Waveform(
      Waveform Count=::NewEnumerator1
      Global Input Position=&MergeHelper_FloatSwitch(0=0, 1=0.6, 2=?, 3=?, Index=$Particles.SpawnGroup, Num=::NewEnumerator0)
      "Waveform [1]"=::NewEnumerator0
      "[1] Input Position"=&MergeHelper_FloatSwitch(0=0, 1=0.8, 2=?, 3=?, Index=$Particles.SpawnGroup, Num=::NewEnumerator0)
      "[1] Frequency"=2
      "[1] Amplitude Min/Max"=<0.8 1>
      "Waveform [2] Blend"=::NewEnumerator2
      "Waveform [2]"=::NewEnumerator1
      "[2] Input Position"=0
      "[2] Phase"=$Emitter.Age
      "[2] Phase Scale"=1
      "[2] Compound Frequency"=2
      "[2] Frequency"=6
      "[2] Amplitude Min/Max"=<0.9 1>
      "[2] Scale"=1
      "[2] Bias"=0
      "[2] Exponent"=1
      "[2] Pulse Duration"=0.1
      "[2] Clip Triangle Min/Max"=<0 0>
      "[2] Random Seed"=0
     )
    )) {isSetParameters=true}
   /SetVariables(
    Particles.TimeOfDayHeadlightsIntensity:NiagaraFloat=&Multiply_Float(
     A=&FloatFromCurve(
      FloatCurve=DataInterface() {propertyValues {Curve {keys [
          {interpMode=RCIM_Cubic, tangentMode=RCTM_Break, time=0, value=0}
          {interpMode=RCIM_Cubic, tangentMode=RCTM_Break, time=0.0714, value=1}
          <at 0.0847 0.2 RCIM_Cubic>
          <at 0.112 1 RCIM_Cubic>
         ]}}}
      Scale Curve=$StackContext.Flicker
     )
     B=&FN_FrontEndFloatModifier(
      Gameplay=&FNTimeOfDayScalar(Night Scale=1, Morning Scale=0.7, Day Scale=0.5, Evening Scale=0.7)
      Front End=0.5
     )
    )
    Particles.TimeOfDayVisibilityTag:NiagaraInt32=&MakeIntFromFloat(Float=&FN_FrontEndFloatModifier(
      Gameplay=&FNTimeOfDayScalar(Night Scale=0, Morning Scale=0, Day Scale=0, Evening Scale=0)
      Front End=0
     ))
   ) {isSetParameters=true}
   /SetVariables(
    Particles.BeamColor:LinearColor=&MultiplyLinearColor(
     Linear Color=$StackContext.BeamColor
     Scale Factor=&MakeVector4_FromFloat(Float=$Particles.TimeOfDayHeadlightsIntensity)
    )
    Particles.Color=&MultiplyLinearColor(
     Linear Color=$StackContext.Color
     Scale Factor=&MakeVector4_FromFloat(Float=$Particles.TimeOfDayHeadlightsIntensity)
    )
    Particles.FlareColor=&MultiplyLinearColor(
     Linear Color=$StackContext.FlareColor
     Scale Factor=&MakeVector4_FromFloat(Float=$Particles.TimeOfDayHeadlightsIntensity)
    )
    Particles.GlowColor=&MultiplyLinearColor(
     Linear Color=$StackContext.GlowColor
     Scale Factor=&MakeVector4_FromFloat(Float=$Particles.TimeOfDayHeadlightsIntensity)
    )
    Particles.LightColor=&MultiplyLinearColor(
     Linear Color=$StackContext.LightColor
     Scale Factor=&MakeVector4_FromFloat(Float=$Particles.TimeOfDayHeadlightsIntensity)
    )
    Particles.TeardropGlowColor:LinearColor=&MultiplyLinearColor(
     Linear Color=$StackContext.TeardropGlowColor
     Scale Factor=&MakeVector4_FromFloat(Float=$Particles.TimeOfDayHeadlightsIntensity)
    )
   ) {isSetParameters=true}
  ]
  rendererClasses [&NiagaraSpriteRendererProperties]
  renderers [
   {rendererClass=&NiagaraSpriteRendererProperties}
   {rendererClass=&NiagaraSpriteRendererProperties, rendererIndex=1}
   {rendererClass=&NiagaraSpriteRendererProperties, rendererIndex=2}
  ]
 }
 SmokeLinger {
  emitterUpdateScript [
   &EmitterState(Life Cycle Mode=::NewEnumerator1, Loop Behavior=::NewEnumerator1, Loop Delay=0.3, Loop Duration Mode=::NewEnumerator1, UseLoopDelay=true)
   &SpawnRate(
    SpawnRate=&FN_FrontEndFloatModifier(
     Gameplay=&Lerp_Float(A=10, B=30, Alpha=$System.Timeline.Time)
     Front End=&Lerp_Float(A=14, B=40, Alpha=$System.Timeline.Time)
    )
    Use Spawn Probability=false
   ) {module=SpawnRate001}
  ]
  particleSpawnScript [
   &FN_SetSpawnGroup(Number of Spawn Groups:NiagaraInt32=&MakeCustomIntFromBool(Bool=&Boolean_OrOperation(Bool1=$User.bIsFrontEnd, Bool2=$User.bIsFrontEndPreview), True INT=8, False INT=6))
   &InitializeParticle(
    Color Maximum=<rgba 0.0417 0.0385 0.0354 1.1>
    Color Minimum=<rgba 0.2448 0.2264 0.2081 1.1>
    Color Mode=::NewEnumerator2
    Lifetime=&Multiply_Float(A=1, B=&Lerp_Float(A=1, B=0.7, Alpha=$System.Timeline.Time))
    Mesh Scale Mode=::NewEnumerator3
    Sprite Rotation Mode=::NewEnumerator2
    Sprite Size Mode=::NewEnumerator1
    Uniform Sprite Size Max=150
    Uniform Sprite Size Min=100
    UsePositionOffset=false
    Write Material Random=true
   )
   /SetVariables(Particles.InitSpriteRotRate=&UniformRangedFloat(Minimum=-3, Maximum=3)) {isSetParameters=true}
   &CameraOffset(Camera Offset Amount=10)
  ]
  particleUpdateScript [
   &ParticleState()
   &SocketLocation(
    Interpolate Socket Sampling=true
    SamplePosOffset=true
    Sampled Position Offset=&Lerp_Vector(
     A=&MakeVector(X=0, Y=0, Z=0)
     B=&MakeVector(
      X=&Multiply_Float(
       A=&MergeHelper_FloatSwitch(0=1, 1=1, 2=1, 3=-1, 4=-1, 5=-1, Index=$Particles.SpawnGroup, Num=::NewEnumerator4)
       B=&FN_FrontEndFloatModifier(Gameplay=&Lerp_Float(A=20, B=5, Alpha=$System.Timeline.Time), Front End=-20)
      )
      Y=&FN_FrontEndFloatModifier(Gameplay=&Lerp_Float(A=25, B=100, Alpha=$System.Timeline.Time), Front End=-80)
      Z=10
     )
     Alpha=$Particles.NormalizedAge
    )
    Sampled Position Offset Coordinate Space:ENiagaraSkelMeshTransforms=::NewEnumerator2
    Socket Index=&MergeHelper_IntSwitch(0=4, 1=5, 2=6, 3=7, 4=8, 5=9, 6=12, 7=12, Index=$Particles.SpawnGroup, Num=::NewEnumerator6)
    Socket Reader=$System.Socket Reader
    Socket Sampling Mode=::NewEnumerator3
    VelocitySpeedLimit=false
   ) {module=SocketLocation001}
   &SpriteRotationRate(Rotation Rate=&Multiply_Float(
     A=$Particles.InitSpriteRotRate
     B=&FloatFromCurve(
      FloatCurve=DataInterface() {propertyValues {Curve {keys [
          {arriveTangent=-5.9085, interpMode=RCIM_Cubic, leaveTangent=-5.9085, tangentMode=RCTM_User, time=0, value=1}
          {arriveTangent=-0.2496, interpMode=RCIM_Cubic, leaveTangent=-0.2496, tangentMode=RCTM_User, time=0.5, value=0.1}
         ]}}}
      CurveIndex=$Emitter.Age
      Scale Curve=150
     )
    ))
   &ScaleColor(
    Scale Alpha=&FloatFromCurve(
     FloatCurve=DataInterface() {propertyValues {Curve {keys [
         {arriveTangent=8.0265, interpMode=RCIM_Cubic, leaveTangent=8.0265, tangentMode=RCTM_User, time=0, value=0}
         <at 0.15 1 RCIM_Cubic>
         {arriveTangent=-3.2327, interpMode=RCIM_Cubic, leaveTangent=-3.2327, tangentMode=RCTM_User, time=1, value=0}
        ]}}}
     Scale Curve=&Multiply_Float(A=1.33, B=&FN_FrontEndFloatModifier(Gameplay=1, Front End=0.7))
    )
    Scale RGB=&VectorFromFloat(Value=&FloatFromCurve(FloatCurve=DataInterface() {propertyValues {Curve {keys [{arriveTangent=1.8145, interpMode=RCIM_Cubic, leaveTangent=4.5795, tangentMode=RCTM_Break, time=0, value=0.4}, <at 0.3387 0.9565>, <at 1 1>]}}}))
    ScaleA=true
    ScaleRGB=true
   )
   /Niagara/Modules/Update/Forces/CurlNoiseForce(Noise Frequency=5, Noise Strength=400, Pan Noise=true, Pan Noise Field=<0 0 1>)
   &Drag(Drag=&RandomRangeFloat(Minimum=0.6, Maximum=1.5), Use Linear Drag=true, Use Rotational Drag=false)
   &SolveForcesAndVelocity(Clamp Velocity=false, Limit Acceleration=false)
   &ScaleSpriteSize(Scale Factor=&Vector2DFromFloat(Value=&FloatFromCurve(
      FloatCurve=DataInterface() {propertyValues {Curve {keys [
          {arriveTangent=1.1957, interpMode=RCIM_Cubic, leaveTangent=1.1957, leaveTangentWeight=0.2993, tangentMode=RCTM_User, time=0, value=0.2}
          {arriveTangent=0.7808, interpMode=RCIM_Cubic, leaveTangent=0.7808, time=0.5406, value=0.7231}
          {
           arriveTangent=0.2119
           arriveTangentWeight=0.3618
           interpMode=RCIM_Cubic
           leaveTangent=0.2119
           tangentMode=RCTM_User
           time=1.1327
           value=1.0844
          }
         ]}}}
      Scale Curve=&Multiply_Float(A=$System.OwnerScale, B=1)
     )))
   &DynamicMaterialParameters(
    Index 0 Param 1=&Add_Float(
     A=&RandomRangeFloat(Minimum=0, Maximum=1)
     B=&FloatFromCurve(
      FloatCurve=DataInterface() {propertyValues {Curve {keys [
          {arriveTangent=2.7533, interpMode=RCIM_Cubic, leaveTangent=1.8958, tangentMode=RCTM_User, time=0, value=1}
          {arriveTangent=0.8292, interpMode=RCIM_Cubic, leaveTangent=0.8292, tangentMode=RCTM_User, time=0.321, value=1.3453}
          {arriveTangent=0.0529, interpMode=RCIM_Cubic, leaveTangent=0.2021, tangentMode=RCTM_User, time=1, value=1.5}
         ]}}}
      CurveIndex=&Add_Float(A=&Multiply_Float(A=$Particles.Initial.MaterialRandom, B=0.5), B=$Particles.NormalizedAge)
     )
    )
    Index 0 Param 2=&FloatFromCurve(FloatCurve=DataInterface() {propertyValues {Curve {keys [
         {arriveTangent=0.0453, interpMode=RCIM_Cubic, tangentMode=RCTM_Break, time=0, value=-1}
         {arriveTangent=0.0604, interpMode=RCIM_Cubic, leaveTangent=0.0604, tangentMode=RCTM_User, time=1, value=-1}
        ]}}})
    Index 0 Param 3=&FloatFromCurve(FloatCurve=DataInterface() {propertyValues {Curve {keys [
         {interpMode=RCIM_Cubic, leaveTangentWeight=0.0333, tangentMode=RCTM_User, time=0, value=0}
         {arriveTangentWeight=0.0391, interpMode=RCIM_Cubic, leaveTangentWeight=0.1173, tangentMode=RCTM_User, time=0.1, value=0}
         {arriveTangentWeight=0.1, interpMode=RCIM_Cubic, tangentMode=RCTM_Break, time=0.4, value=0}
        ]}}})
    Index 0 Param 4=&FloatFromCurve(FloatCurve=DataInterface() {propertyValues {Curve {keys [<at 0.5 1>,<at 1 0>]}}})
    Index 1 Param 1=&FloatFromCurve(
     FloatCurve=DataInterface() {propertyValues {Curve {keys [
         {interpMode=RCIM_Cubic, tangentMode=RCTM_Break, time=0, value=1}
         {arriveTangentWeight=0.0833, interpMode=RCIM_Cubic, tangentMode=RCTM_Break, time=0.25, value=0.3}
        ]}}}
     Scale Curve=&RandomRangeFloat(Minimum=1.5, Maximum=3)
    )
    Index 1 Param 2=&FloatFromCurve(
     FloatCurve=DataInterface() {propertyValues {Curve {keys [
         {interpMode=RCIM_Cubic, leaveTangent=-2.5489, tangentMode=RCTM_Break, time=0, value=1}
         {interpMode=RCIM_Cubic, leaveTangent=0.9066, tangentMode=RCTM_Break, time=1, value=0}
        ]}}}
     Scale Curve=&RandomRangeFloat(Minimum=0.01, Maximum=0.1)
    )
    Index 1 Param 3=&FloatFromCurve(FloatCurve=DataInterface() {propertyValues {Curve {keys [
         {interpMode=RCIM_Cubic, leaveTangent=6.3917, tangentMode=RCTM_Break, time=0, value=0.2}
         {arriveTangent=1.7549, interpMode=RCIM_Cubic, leaveTangent=10.6767, tangentMode=RCTM_User, time=0.25, value=1}
        ]}}})
    Index 1 Param 4=&FloatFromCurve(
     FloatCurve=DataInterface() {propertyValues {Curve {keys [{arriveTangent=-0.136, interpMode=RCIM_Cubic, leaveTangent=-0.136, tangentMode=RCTM_User, time=0, value=1}, <at 1 0.95>]}}}
     Scale Curve=&RandomRangeFloat(Minimum=1.3, Maximum=2)
    )
    Index 2 Param 1=&FloatFromCurve(FloatCurve=DataInterface() {propertyValues {Curve {keys [
         {interpMode=RCIM_Cubic, leaveTangent=4.2485, tangentMode=RCTM_Break, time=0, value=0}
         {arriveTangent=1.0548, interpMode=RCIM_Cubic, leaveTangent=1.0548, tangentMode=RCTM_User, time=0.17, value=0.5715}
         <at 1 1>
        ]}}})
    Index 2 Param 2=&FloatFromCurve(
     FloatCurve=DataInterface() {propertyValues {Curve {keys [
         {arriveTangent=1.9056, interpMode=RCIM_Cubic, leaveTangent=1.9056, tangentMode=RCTM_User, time=0, value=0.2}
         {arriveTangent=0.0259, interpMode=RCIM_Cubic, leaveTangent=0.0259, tangentMode=RCTM_User, time=1, value=1.4}
        ]}}}
     Scale Curve=&RandomRangeFloat(Minimum=0.1, Maximum=0.2)
    )
    Index 2 Param 3=&FloatFromCurve(FloatCurve=DataInterface() {propertyValues {Curve {keys [
         {interpMode=RCIM_Cubic, leaveTangent=2.3064, tangentMode=RCTM_Break, time=0, value=1}
         {interpMode=RCIM_Cubic, leaveTangent=11.1233, tangentMode=RCTM_Break, time=1, value=2}
        ]}}})
    Index 2 Param 4=&FloatFromCurve(
     FloatCurve=DataInterface() {propertyValues {Curve {keys [
         {arriveTangent=-0.6867, interpMode=RCIM_Cubic, leaveTangent=-0.6867, tangentMode=RCTM_User, time=0, value=1}
         {interpMode=RCIM_Cubic, leaveTangent=-1.2956, tangentMode=RCTM_Break, time=1, value=0.75}
        ]}}}
     Scale Curve=1.5
    )
    Module.Param0WriteEnabled=true
    Module.Param0WriteEnabled1=true
    Module.Param0WriteEnabled2=true
    Module.Param1WriteEnabled=false
    Module.Param1WriteEnabled1=true
    Module.Param1WriteEnabled2=true
    Module.Param2WriteEnabled=true
    Module.Param2WriteEnabled1=true
    Module.Param2WriteEnabled2=true
    Module.Param3WriteEnabled=true
    Module.Param3WriteEnabled1=true
    Module.Param3WriteEnabled2=true
    Write Parameter Index 1=true
    Write Parameter Index 2=true
   )
   &ScaleColor(
    Scale Alpha=&FN_PickFloatDependOnPlatform(TargetPlatform=DataInterface() {propertyValues {Platforms {qualityLevelMask=-31}}}, TargetPlatformActive=1.5, TargetPlatformNoActive=1)
    Scale RGB=&MakeVectorFromLinearColorRGB(Color=&LerpLinearColors(
      StartColor=<rgba 1 1 1 1>
      EndColor=&LerpLinearColors(
       StartColor=<rgba 1 1 1 1>
       EndColor=&LerpLinearColors(StartColor=<rgba 1 0.37 0.3 1>, EndColor=<rgba 1 0.3 0.7559 1>, LerpFactor=$User.Variant)
       LerpFactor=&FloatFromCurve(FloatCurve=DataInterface() {propertyValues {Curve {keys [<at 0.6 0>,<at 0.9 1>]}}})
      )
      LerpFactor=$System.Timeline.Time
     ))
    ScaleA=true
    ScaleRGB=true
   ) {module=ScaleColor001}
  ]
  rendererClasses [&NiagaraSpriteRendererProperties]
  renderers [
   {rendererClass=&NiagaraSpriteRendererProperties}
   {rendererClass=&NiagaraSpriteRendererProperties, rendererIndex=1}
  ]
 }
 Sparks {
  emitterUpdateScript [
   &EmitterState(
    Life Cycle Mode=::NewEnumerator1
    Loop Behavior=::NewEnumerator1
    Loop Delay=$System.InitialThursterDelay
    Loop Duration Mode=::NewEnumerator1
    UseLoopDelay=true
   )
   &SpawnRate(
    Spawn Probability=&FloatFromCurve(
     FloatCurve=DataInterface() {propertyValues {Curve {keys [<at 0 0 RCIM_Constant>, <at 0.05 1>, <at 0.2513 1>, {arriveTangent=-2.5249, leaveTangent=-2.5249, time=0.5, value=0}]}}}
     CurveIndex=$System.Timeline.NormalizedTime
     Scale Curve=&FN_FrontEndFloatModifier(Gameplay=$User.ForwardDot, Front End=1)
    )
    SpawnRate=500
    Use Spawn Probability=true
   )
   &SpawnRate(
    Spawn Probability=&FN_FrontEndFloatModifier(Gameplay=$System.Timeline.Time, Front End=1)
    SpawnRate=30
    Use Spawn Probability=true
   ) {module=SpawnRate001}
  ]
  particleSpawnScript [
   &FN_SetSpawnGroup(Number of Spawn Groups=2)
   /SetVariables(Particles.PerParticleRandom:NiagaraFloat=&RandomRangeFloat(Minimum=0, Maximum=1)) {isSetParameters=true}
   &InitializeParticle(
    Color Mode=::NewEnumerator0
    Lifetime=&Lerp_Float(A=0.3, B=0.6, Alpha=$Particles.PerParticleRandom)
    Sprite Size=&Multiply_Vector2D_ByFloat(Float=1.4, Vector2D=&Lerp_Vector2(A=<1 1.5>, B=<2 4>, Alpha=$Particles.PerParticleRandom))
    Sprite Size Mode=::NewEnumerator4
    UsePositionOffset=false
    Write Material Random=true
   )
   &SocketLocation(
    Interpolate Socket Sampling=true
    SamplePosOffset=false
    Sampled Orientation Coordinate Space:ENiagaraCoordinateSpace=::Local
    Sampled Velocity Coordinate Space:ENiagaraCoordinateSpace=::Local
    Socket Index=$Particles.SpawnGroup
    Socket Reader=$System.Socket Reader
    Socket Sampling Mode=::NewEnumerator2
    VelocitySpeedLimit=false
   )
   &ShapeLocation(MaskSpawnGroup=false, Override Seed=false, Sphere Radius=10, UseOwnerScale=true)
   &AddVelocity(
    Cone Angle=0
    Cone Axis=&FN_FrontEndVector3Modifier(Gameplay=<0 1 0>, Front End=<0 -1 0>)
    Inner Cone Angle=20
    Module.UseConeFalloff=false
    Module.UseVelDistribution=false
    Override Seed=false
    Velocity Mode=::NewEnumerator2
    Velocity Speed=&Lerp_Float(A=500, B=999, Alpha=$Particles.PerParticleRandom)
   )
  ]
  particleUpdateScript [
   &ParticleState()
   &Drag(Drag=&RandomRangeFloat(Minimum=2, Maximum=4), Use Linear Drag=true, Use Rotational Drag=false) {module=Drag001}
   &CurlNoiseForce(
    Noise Frequency=0.04
    Noise Strength=&FloatFromCurve(
     FloatCurve=DataInterface() {propertyValues {Curve {keys [{tangentMode=RCTM_User, time=0, value=0}, {tangentMode=RCTM_User, time=1, value=1}]}}}
     Scale Curve=&Lerp_Float(A=2222, B=222, Alpha=$Particles.PerParticleRandom)
    )
    Pan Noise=true
    Pan Noise Field=<0 0 1>
   )
   &InheritSourceMovement(
    Applied Position Scale=&VectorFromFloat(Value=&FloatFromCurve(FloatCurve=DataInterface() {propertyValues {Curve {keys [<at 0.8 1 RCIM_Cubic>,<at 1 0>]}}}))
    ApplyForce=false
    ApplyPosition=true
    ApplyPositionFalloff=false
   )
   &SolveForcesAndVelocity(Clamp Velocity=false, Limit Acceleration=false)
   &ScaleColor(
    Scale Alpha=&FloatFromCurve(FloatCurve=DataInterface() {propertyValues {Curve {keys [<at 0.9198 1>,<at 1 0>]}}})
    Scale RGB=&MakeVectorFromLinearColorRGB(Color=&LerpLinearColors(
      StartColor=&ColorFromCurve(ColorCurve=DataInterface() {propertyValues {
         AlphaCurve {keys [<at 0.01 1>,<at 0.9198 1 RCIM_Cubic>,<at 1 0>]}
         BlueCurve {keys [<at 0 15>,<at 0.0135 0.5>,<at 0.2072 0>,<at 0.4962 0>,<at 0.9394 0>]}
         GreenCurve {keys [<at 0 5>,<at 0.0135 3.8>,<at 0.2072 2.2>,<at 0.4962 0.1833>,<at 0.9394 0.0762>]}
         RedCurve {keys [<at 0 0>,<at 0.0135 5>,<at 0.2072 3>,<at 0.4962 1>,<at 0.9394 1>]}
        }})
      EndColor=&ColorFromCurve(ColorCurve=DataInterface() {propertyValues {
         AlphaCurve {keys [<at 0.01 1>,<at 0.9198 1 RCIM_Cubic>,<at 1 0>]}
         BlueCurve {keys [<at 0.0251 10.5>,<at 0.2072 7.3333>,<at 0.5571 1>]}
         GreenCurve {keys [<at 0.0251 1.5>,<at 0.2072 2>,<at 0.5571 0>]}
         RedCurve {keys [<at 0.0251 15>,<at 0.2072 10>,<at 0.5571 0.6667>]}
        }})
      LerpFactor=$User.Variant
     ))
    ScaleA=true
    ScaleRGB=true
   ) {module=ScaleColor002}
   &ScaleColor(Scale Alpha=0.1, ScaleA=true, ScaleRGB=false) {module=ScaleColor001}
   &ScaleSpriteSizeBySpeed(
    Max Scale Factor=<0.7 3>
    Sample Scale Factor By Curve=true
    Scale Factor Curve=DataInterface() {propertyValues {Curve {keys [<at 0 0>, {arriveTangent=1.4989, interpMode=RCIM_Cubic, tangentMode=RCTM_Break, time=1, value=1}]}}}
    Velocity Threshold=999
   )
   &ScaleSpriteSize(
    Uniform Curve Scale=&MakeFloatFromVector(Vector=$Engine.Owner.Scale, Channel=::NewEnumerator0)
    Uniform Curve Sprite Scale=DataInterface() {propertyValues {Curve {keys [
        {tangentMode=RCTM_User, time=0, value=1}
        {arriveTangent=-0.0405, interpMode=RCIM_Cubic, leaveTangent=-0.0405, tangentMode=RCTM_User, time=0.7009, value=1}
        {arriveTangent=-8.0089, interpMode=RCIM_Cubic, leaveTangent=-8.0089, tangentMode=RCTM_User, time=1, value=0}
       ]}}}
   )
  ]
  rendererClasses [&NiagaraSpriteRendererProperties]
  renderers [{rendererClass=&NiagaraSpriteRendererProperties}]
 }
 Speedlines {
  emitterUpdateScript [
   &EmitterState()
   &SpawnRate(
    Spawn Probability=&MakeCustomFloatFromBool(
     Bool=&SetBoolByFloatComparison(A=$User.ForwardDot, B=0.05)
     True Float=1
     False Float=&FN_FrontEndFloatModifier(Gameplay=0, Front End=1)
    )
    SpawnRate=&MakeCustomFloatFromBool(Bool=$User.bIsFrontEndPreview, True Float=50, False Float=90)
    Use Spawn Probability=true
   ) {module=SpawnRate001}
  ]
  particleSpawnScript [
   &InitializeParticle(
    Color=&MakeLinearColorFromVectorAndFloat(
     "Vector (RGB)"=<1 1 1>
     "Float (Alpha)"=&Multiply_Float(
      A=&UniformRangedFloat(Minimum=1, Maximum=1)
      B=&MakeCustomFloatFromBool(
       Bool=$User.bIsFrontEndPreview
       True Float=&FloatFromCurve(
        FloatCurve=DataInterface() {propertyValues {Curve {keys [{tangentMode=RCTM_User, time=0, value=0}, {tangentMode=RCTM_User, time=2, value=1}]}}}
        CurveIndex=$Emitter.Age
        Scale Curve=0.4
       )
       False Float=1
      )
     )
    )
    Lifetime Max=0.2
    Lifetime Min=0.15
    Lifetime Mode=::NewEnumerator1
    Position Offset=&FN_FrontEndVector3Modifier(Gameplay=<0 -50 0>, Front End=<0 50 0>)
    Sprite Size Max=<3 10>
    Sprite Size Min=<2 2>
    Sprite Size Mode=::NewEnumerator2
    UsePositionOffset=true
    Write Material Random=true
   )
   &AddVelocity(
    Module.Use Falloff Distance=false
    Override Seed=false
    Velocity Mode=::NewEnumerator1
    Velocity Speed=&FN_FrontEndFloatModifier(Gameplay=-2500, Front End=2500)
   ) {module=AddVelocity001}
   &TorusLocation(
    Align Torus With Axis=true
    Coordinate Space:ENiagaraCoordinateSpace=::Simulation
    Handle Radius=0
    Large Radius:NiagaraFloat=&FN_FrontEndFloatModifier(
     Gameplay=&RandomRangeFloat(Minimum=450, Maximum=500)
     Front End=&RandomRangeFloat(Minimum=350, Maximum=400)
    )
    MaskSpawnGroup=false
    Module.Override Randomness=false
    Offset=<0 33 0>
    Torus Axis=<0 1 0>
    Torus Origin:NiagaraPosition=&AddVectorToPosition(
     Position=$Particles.Position
     Vector=&FN_FrontEndVector3Modifier(Gameplay=<0 -500 0>, Front End=<0 -500 0>)
    )
   )
   &AddVelocity(
    Velocity=&Add_Vector(
     A=&RandomRangeVector(Minimum=<0 300 0>, Maximum=<0 700 1>)
     B=&Add_Vector(
      A=&Multiply_VectorByFloat(Vector=<-555 0 0>, Float=$User.RightDot)
      B=&Multiply_VectorByFloat(Vector=<0 0 222>, Float=$User.ForwardDot)
     )
    )
    Velocity Speed Scale=&FN_FrontEndFloatModifier(Gameplay=1, Front End=-1)
   )
  ]
  particleUpdateScript [
   &SolveForcesAndVelocity(Clamp Velocity=false, Limit Acceleration=false)
   &ParticleState()
   &ScaleSpriteSize(Scale Factor=&Vector2DFromCurve(
     Vector2Curve=DataInterface() {propertyValues {XCurve {keys [<at 0 0.5 RCIM_Cubic>,<at 0.5 1 RCIM_Cubic>]}, YCurve {keys [<at 0 0 RCIM_Cubic>,<at 0.5 1 RCIM_Cubic>]}}}
     ScaleCurve=&Vector2DFromFloat(Value=1)
    ))
   &ScaleSpriteSizeBySpeed(Max Scale Factor=<1 50>, Min Scale Factor=<0.01 0.05>, Velocity Threshold=1)
   &ScaleColor(
    Scale Alpha=&FloatFromCurve(
     FloatCurve=DataInterface() {propertyValues {Curve {keys [
         <at 0 0 RCIM_Cubic>
         <at 0.3759 1.0207>
         {arriveTangent=-2.221, interpMode=RCIM_Cubic, leaveTangent=-2.221, tangentMode=RCTM_User, time=1, value=0}
        ]}}}
     Scale Curve=&MakeCustomFloatFromBool(
      Bool=$User.bIsFrontEndPreview
      True Float=0.3759
      False Float=&Lerp_Float(A=0, B=0.5, Alpha=&FN_FrontEndFloatModifier(Gameplay=$System.Timeline.Time, Front End=0.5))
     )
    )
    Scale RGB=&VectorFromFloat(Value=&FN_FrontEndFloatModifier(Gameplay=0.5, Front End=0.1))
    ScaleA=true
    ScaleRGB=true
   )
  ]
  rendererClasses [&NiagaraSpriteRendererProperties]
  renderers [{rendererClass=&NiagaraSpriteRendererProperties}]
 }
 Thursters {
  emitterUpdateScript [
   &EmitterState(
    Inactive Response=::NewEnumerator1
    Life Cycle Mode=::NewEnumerator1
    Loop Behavior=::NewEnumerator1
    Loop Delay=$System.InitialThursterDelay
    Loop Duration Mode=::NewEnumerator1
    UseLoopDelay=true
   )
   &SpawnBurst_Instantaneous(
    Spawn Count=2
    Spawn Probability=HlslExpression() {hlslExpression="Emitter.LoopCount == 0"}
    Use Spawn Probability=false
    UseLoopCountLimit=false
   )
  ]
  particleSpawnScript [
   &FN_SetSpawnGroup(Number of Spawn Groups=2)
   &InitializeParticle(Color Mode=::NewEnumerator0, Mesh Scale=<2 2 2>, Mesh Scale Mode=::NewEnumerator4, UsePositionOffset=false, Write Material Random=true)
   &InitialMeshOrientation(Mesh Orientation Mode=::NewEnumerator2, Orientation Vector=<0 0 1>, Use Rotation Vector=false)
   /SetVariables() {isSetParameters=true}
   /SetVariables(Particles.ThrusterVisibilityTag:NiagaraInt32=&MergeHelper_IntSwitch(0=0, 1=1, 2=?, 3=?, 4=?, 5=?, Index=&MakeIntFromFloat(Float=$User.Variant), Num=::NewEnumerator0)) {isSetParameters=true}
  ]
  particleUpdateScript [
   &ParticleState(Kill Particles When Lifetime Has Elapsed=false, Let Infinitely Lived Particles Die When Emitter Deactivates=true)
   /SetVariables(Particles.Jitter:NiagaraFloat=&Waveform(
     Global Input Position=&RandomRangeFloat(Minimum=0, Maximum=1, Evaluation Type=::NewEnumerator1)
     "Waveform [1]"=::NewEnumerator12
     "[1] Frequency"=20
     "[1] Amplitude Min/Max"=<0.7 1.1>
    )) {isSetParameters=true}
   &SocketLocation(
    SamplePosOffset=true
    Socket Index=$Particles.SpawnGroup
    Socket Reader=$System.Socket Reader
    Socket Sampling Mode=::NewEnumerator3
    VelocitySpeedLimit=false
   )
   &ScaleMeshSize(Scale Factor=&Multiply_Vector(
     A=<0.6 1 1>
     B=&Multiply_Vector(
      A=&Lerp_Vector(A=<1 1 1>, B=<1.8 1.2 1.2>, Alpha=$System.Timeline.NormalizedTime)
      B=&Lerp_Vector(A=<0.3 1 1>, B=<1 1 1>, Alpha=$System.Timeline001.Time)
     )
    ))
   /SetVariables(Particles.RibbonTime:NiagaraFloat=&Add_Float(
     A=$Particles.RibbonTime
     B=&Multiply_Float(
      A=$Engine.DeltaTime
      B=&FloatFromCurve(
       FloatCurve=DataInterface() {propertyValues {Curve {keys [{interpMode=RCIM_Cubic, tangentMode=RCTM_Break, time=0, value=1.5}, {interpMode=RCIM_Cubic, tangentMode=RCTM_Break, time=1, value=2}]}}}
       CurveIndex=$System.Timeline.Time
      )
     )
    )) {isSetParameters=true}
   &ScaleColor(
    Scale Alpha=$Particles.Jitter
    Scale RGB=&VectorFromFloat(Value=$Particles.Jitter)
    ScaleA=true
    ScaleRGB=true
   )
   &ScaleColor(
    Scale RGB=&VectorFromFloat(Value=&Lerp_Float(
      A=&FloatFromCurve(
       FloatCurve=DataInterface() {propertyValues {Curve {keys [<at 0 1 RCIM_Cubic>,<at 0.0302 15 RCIM_Cubic>,<at 0.5 1 RCIM_Cubic>]}}}
       CurveIndex=$System.Timeline002.Time
      )
      B=&FloatFromCurve(
       FloatCurve=DataInterface() {propertyValues {Curve {keys [<at 0 1 RCIM_Cubic>,<at 0.0302 5 RCIM_Cubic>,<at 0.5 1 RCIM_Cubic>]}}}
       CurveIndex=$System.Timeline002.Time
      )
      Alpha=$User.Variant
     ))
    ScaleA=false
    ScaleRGB=true
   ) {module=ScaleColor001}
   &ScaleMeshSize(Scale Factor=&Lerp_Vector(
     A=<0 0 0>
     B=<1 0.85 0.85>
     Alpha=&FloatFromCurve(FloatCurve=DataInterface() {propertyValues {Curve {keys [
          <at 0 0 RCIM_Cubic>
          {interpMode=RCIM_Cubic, leaveTangent=-4.9031, tangentMode=RCTM_Break, time=0.05, value=1.5}
          {interpMode=RCIM_Cubic, tangentMode=RCTM_Break, time=0.3, value=1}
         ]}}})
    )) {module=ScaleMeshSize002}
   &FN_Sprite_Attributes(
    Alignment:Vector3f=<1 0 0>
    Color=&LerpLinearColors(StartColor=<rgba 2 0.6067 0.1 1>, EndColor=<rgba 2 0.1 1.3667 1>, LerpFactor=$User.Variant)
    Create Alignment Attribute=true
    Create Camera Offset Attribute=true
    Create Color Attribute=true
    Create Facing Attribute=true
    Create Position Attribute=true
    Create SpriteSize Attribute=true
    Facing:Vector3f=&FN_FrontEndVector3Modifier(Gameplay=<0 1 0>, Front End=<0 -1 0>)
    Position:NiagaraPosition=&AddVectorToPosition(
     Position=$Output.SocketLocation.SampledPosition
     Vector=&FN_FrontEndVector3Modifier(Gameplay=<0 1 0>, Front End=<0 -1 0>)
    )
    SpriteSize=&Multiply_Vector2D_ByFloat(
     Float=&Multiply_Float(
      A=1
      B=&FloatFromCurve(
       FloatCurve=DataInterface() {propertyValues {Curve {keys [{interpMode=RCIM_Cubic, tangentMode=RCTM_Break, time=0.05, value=1.6}, {interpMode=RCIM_Cubic, tangentMode=RCTM_Break, time=0.3, value=1}]}}}
       Scale Curve=&FN_FrontEndFloatModifier(Gameplay=1, Front End=0.7)
      )
     )
     Vector2D=<25 25>
    )
    Write to Module Namespace=true
   )
   &FN_Sprite_Attributes(
    Camera Offset:NiagaraFloat=&FN_FrontEndFloatModifier(Gameplay=100, Front End=0)
    Color=&MultiplyLinearColor(
     Linear Color=&LerpLinearColors(StartColor=<rgba 3 0.72 0.15 1>, EndColor=<rgba 3 0.15 2.05 1>, LerpFactor=$User.Variant)
     Scale Factor=&MakeVector4FromVectorAndFloat(
      "Vector (RGB)"=<1 1 1>
      "Float (Alpha)"=&Multiply_Float(
       A=&Add_Float(
        A=&Add_Float(
         A=&Multiply_Float(A=$System.PushCurve, B=2)
         B=&Multiply_Float(A=$System.Timeline.Time, B=0.3)
        )
        B=0.2
       )
       B=$Particles.Jitter
      )
     )
    )
    Create Camera Offset Attribute=true
    Create Color Attribute=true
    Create Position Attribute=true
    Create SpriteSize Attribute=true
    Position:NiagaraPosition=&AddVectorToPosition(
     Position=$Output.SocketLocation.SampledPosition
     Vector=&FN_FrontEndVector3Modifier(Gameplay=<0 0 0>, Front End=<0 -10 0>)
    )
    SpriteSize=<150 100>
    Write to Module Namespace=true
   ) {module=FN_Sprite_Attributes003}
   &FN_Light_Attributes(
    Color=&Multiply_LinearColorByFloat(
     Linear Color=&LerpLinearColors(StartColor=<rgba 1 0.145 0.05 1>, EndColor=<rgba 1 0.1 0.7 1>, LerpFactor=$User.Variant)
     Float=&Multiply_Float(
      A=&Lerp_Float(A=0.7, B=2, Alpha=$System.Timeline.Time)
      B=&Lerp_Float(A=0.3, B=1, Alpha=$System.Timeline001.Time)
     )
    )
    Create Color Attribute=true
    Create Exponent Attribute=true
    Create Position Attribute=true
    Create Radius Attribute=true
    Exponent=3
    Position=&AddVectorToPosition(
     Position=$Output.SocketLocation.SampledPosition
     Vector=&FN_FrontEndVector3Modifier(
      Gameplay=&SelectVectorFromArray(Direct Array Index=$Particles.SpawnGroup, Vector Selection Array=DataInterface() {propertyValues {FloatData [<0 30 10>, <0 30 10>]}})
      Front End=&SelectVectorFromArray(Direct Array Index=$Particles.SpawnGroup, Vector Selection Array=DataInterface() {propertyValues {FloatData [<0 -30 10>, <0 -30 10>]}})
     )
    )
    Radius=60
    Scale Color=&Multiply_Float(A=$Particles.Jitter, B=70)
   )
   /SetVariables(Particles.DynParam:Vector4f=&MakeVector4(X=$Particles.RibbonTime, Y=0, Z=0, W=0)) {isSetParameters=true}
  ]
  rendererClasses [
   &NiagaraMeshRendererProperties
   &NiagaraSpriteRendererProperties
   &NiagaraLightRendererProperties
  ]
  renderers [
   {rendererClass=&NiagaraMeshRendererProperties, rendererIndex=0}
   {rendererClass=&NiagaraMeshRendererProperties, rendererIndex=1}
   {rendererClass=&NiagaraSpriteRendererProperties, rendererIndex=2}
   {rendererClass=&NiagaraSpriteRendererProperties, rendererIndex=3}
   {rendererClass=&NiagaraLightRendererProperties, rendererIndex=4}
   {rendererClass=&NiagaraMeshRendererProperties, rendererIndex=5}
  ]
 }
 WindCone {
  emitterUpdateScript [
   &EmitterState(Life Cycle Mode=::NewEnumerator1, Loop Behavior=::NewEnumerator1, Loop Duration Mode=::NewEnumerator1, UseLoopDelay=false)
   &SpawnBurst_Instantaneous(Use Spawn Probability=false, UseLoopCountLimit=false)
  ]
  particleSpawnScript [
   &InitializeParticle(
    Color=<rgba 1 1 1 2>
    Lifetime=0.35
    Mesh Scale=&Multiply_VectorByFloat(Vector=<1.8 1.5 1.5>, Float=1)
    Mesh Scale Mode=::NewEnumerator4
    Position Offset=<0 -150 -50>
    Sprite Size Mode=::NewEnumerator3
    UsePositionOffset=false
    Write Material Random=true
   )
   &InitialMeshOrientation(
    Orientation Axis=<-1 0 0>
    Orientation Coordinate Space=::Local
    Rotation=&FN_FrontEndVector3Modifier(Gameplay=<-0.1 0 0>, Front End=<0.4 0 0>)
    Rotation Coordinate Space=::NewEnumerator2
    Use Orientation Vector=true
    Use Rotation Vector=true
   )
   &CameraOffset(Camera Offset Amount=30)
  ]
  particleUpdateScript [
   &ParticleState(Kill Particles When Lifetime Has Elapsed=false, Let Infinitely Lived Particles Die When Emitter Deactivates=true)
   &SocketLocation(
    Orientation Sampling=::NewEnumerator1
    SamplePosOffset=true
    Sampled Position Offset=&FN_FrontEndVector3Modifier(Gameplay=<0 -160 -70>, Front End=<0 160 -100>)
    Sampled Position Offset Coordinate Space:ENiagaraSkelMeshTransforms=::NewEnumerator2
    Socket Index=11
    Socket Reader=$System.Socket Reader
    Socket Sampling Mode=::NewEnumerator3
    VelocitySpeedLimit=false
   ) {module=SocketLocation001}
   &ScaleMeshSize()
   &ScaleColor(
    Scale Alpha=&Multiply_Float(
     A=&MakeCustomFloatFromBool(
      Bool=$User.bIsFrontEndPreview
      True Float=&MakeCustomFloatFromBool(
       Bool=$User.bIsFullyDeployed
       True Float=&Lerp_Float(
        A=0
        B=0.25
        Alpha=&FloatFromCurve(
         FloatCurve=DataInterface() {propertyValues {Curve {defaultValue=1, keys [{tangentMode=RCTM_User, time=2.25, value=0}, {tangentMode=RCTM_User, time=3, value=1}]}}}
         CurveIndex=$Engine.System.Age
         Scale Curve=0.4
        )
       )
       False Float=0
      )
      False Float=&Lerp_Float(A=0, B=1, Alpha=$System.Timeline.Time)
     )
     B=$System.TOD
    )
    Scale RGB=&Lerp_Vector(
     A=&MakeVectorFromLinearColorRGB(Color=$System.TODColor)
     B=&MakeVectorFromLinearColorRGB(Color=$Particles.Color)
     Alpha=0.1
    )
    ScaleA=true
    ScaleRGB=true
   )
  ]
  rendererClasses [&NiagaraMeshRendererProperties]
  renderers [{rendererClass=&NiagaraMeshRendererProperties, rendererIndex=0}]
 }
}
summary {
 systemName=NS_Glider_DerangedMile
 emitters [
  {emitterName=Headlights, rendererClasses [&NiagaraSpriteRendererProperties]}
  {
   emitterName=Thursters
   rendererClasses [
    &NiagaraMeshRendererProperties
    &NiagaraSpriteRendererProperties
    &NiagaraLightRendererProperties
   ]
  }
  {emitterName=Sparks, rendererClasses [&NiagaraSpriteRendererProperties]}
  {emitterName=Glows, rendererClasses [&NiagaraSpriteRendererProperties]}
  {
   emitterName=ElectricalSparks
   rendererClasses [&NiagaraSpriteRendererProperties, &NiagaraLightRendererProperties]
  }
  {emitterName=SmokeLinger, rendererClasses [&NiagaraSpriteRendererProperties]}
  {emitterName=GliderChunks, rendererClasses [&NiagaraMeshRendererProperties]}
  {emitterName=Speedlines, rendererClasses [&NiagaraSpriteRendererProperties]}
  {emitterName=WindCone, rendererClasses [&NiagaraMeshRendererProperties]}
 ]
 userVariables [
  User.bIsFrontEnd {type=NiagaraBool}
  User.bIsFrontEndPreview {type=NiagaraBool}
  User.bIsFullyDeployed {type=NiagaraBool}
  User.ForwardDot {type=NiagaraFloat, defaultValue=0, description=""}
  User.RightDot {type=NiagaraFloat, defaultValue=0, description=""}
  User.Variant {type=NiagaraFloat, defaultValue=0, description=""}
 ]
}
